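; Pizza variant of the GetAway Valet basic car vehicle, which adds handling for a pizza attachment component.
; On explosion it swaps the attachment to the destroyed material, and on dissolve it drives that material's Fade parameter, mirroring the Taxi approach before deferring to the parent.
; It also appends the attachment to the parent's list of attachment mesh components.
; BeginPlay, overlap, tick and construction add no logic of their own.
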(event EventBeginPlay
  (|CallParentFunction) ; This node is disabled and will not be called.\nDrag off pins to build functionality.
)

(event Collision|EventActorBeginOverlap
  (params (OtherActor :type Actor%20Object%20Reference))
  (|CallParentFunction OtherActor) ; This node is disabled and will not be called.\nDrag off pins to build functionality.
)

(event EventTick
  (params (DeltaSeconds :type Float%20%28single-precision%29))
  (|CallParentFunction DeltaSeconds) ; This node is disabled and will not be called.\nDrag off pins to build functionality.
)

(event EventExplodeCosmetics
  (params (DestroyedActor :type Actor%20Object%20Reference))
  ;[1 1 1] Set Material and fade out attachment - mirrored from Taxi
  (Rendering|Material|SetMaterial (Variables|Default|GetPizza_Attachments) 0 "/Valet/Materials/M_FN_Valet_Destroyed_Master_InstAdjusted.M_FN_Valet_Destroyed_Master_InstAdjusted")
  (|CallParentFunction DestroyedActor))

(event EventDisolveVehicle
  (params (Fade :type Float%20%28double-precision%29))
  ;[1 1 1] Set Material and fade out attachment - mirrored from Taxi
  (Rendering|Material|SetScalarParameterValueonMaterials (Variables|Default|GetPizza_Attachments) "Fade" Fade)
  (|CallParentFunction Fade))

(fn getAttachmentMeshComponents
  (Utilities|Array|Clear (LocalVariables|GetAttachmentMeshComponents))
  (bind _meshes (|CallParentFunction))
  (bind _output_get (LocalVariables|SetAttachmentMeshComponents _meshes))
  (Utilities|Array|AddUnique _output_get (Variables|Default|GetPizza_Attachments))
  (return (LocalVariables|GetAttachmentMeshComponents)))

(fn ConstructionScript
  (|CallParentFunction))
